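# Lit skeletal-mesh material for the OverLoad shotgun's ammo indicator lights.
# Vertex-color red separates light segments, and a SphereMask offset by AmmoCount selects which are lit.
# Sign/Saturate thresholds pick per-segment color and intensity: sine-pulsed AmmoColor with its own brightness for the lowest band, steady AmmoColor between, pulsed PoweredAmmoColor for the highest.
# Color, mask and intensity multiply with an inverse-Fresnel falloff that dims grazing edges.
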Material /ForbiddenFruitWeaponsGameplay/Gameplay/Weapons/OverLoadShotgun/M_Shotgun_OverLoad_Lights_BB
graph {
 properties {bUsedWithMeshDeformer=true, bUsedWithSkeletalMesh=true, pixelDepthOffsetMode=PDOM_Legacy, usedShadingModels=Default Lit}
 nodes [
  Constant()
  Constant()
  Constant() {r=0.5}
  n1 = VectorParameter() {parameterName=AmmoColor, defaultValue=<rgba 0 0.0218 1 1>}
  n2 = ScalarParameter() {defaultValue=1, parameterName=Redlight_PuslingSpeed}
  n3 = LinearInterpolate(
   A=@n1.RGB
   Alpha=MaterialFunctionCall(Speed=Subtract(A=@n2)) {materialFunction=&DebugTimeSine}.Result
  ) {constAlpha=0, constB=0}
  n4 = VertexColor()
  n5 = Constant() {r=1}
  n6 = If(A=@n4.R, B=Constant() {r=0.1}, "A > B"=@n4.R, "A == B"=@n5, "A < B"=@n5)
  n7 = Subtract(A=@n6) {constB=0.1}
  n8 = Saturate(None=Sign(None=Subtract(A=@n7) {constB=0.25}))
  n9 = Sign(None=Subtract(A=@n7) {constB=0.5})
  n10 = VectorParameter() {parameterName=PoweredAmmoColor, defaultValue=<rgba 0 0.4984 1 1>}
  n11 = Subtract(A=@n7) {constB=0.75}
  n12 = LinearInterpolate(
   A=LinearInterpolate(A=LinearInterpolate(A=@n3, B=@n1.RGB, Alpha=@n8), B=@n1.RGB, Alpha=Saturate(None=@n9))
   B=LinearInterpolate(
    A=@n10.RGB
    Alpha=MaterialFunctionCall(Speed=@n2) {materialFunction=&DebugTimeSine}.Result
   ) {constB=0}
   Alpha=Saturate(None=Sign(None=@n11))
  )
  Multiply(A=@n12) {constB=0.05}
  n13 = Constant() {r=0.25}
  n14 = Subtract(A=@n6, B=@n13)
  n15 = Multiply(A=@n13, B=ScalarParameter() {defaultValue=4, parameterName=AmmoCount})
  n16 = SphereMask(A=Saturate(None=Add(A=@n14, B=@n15)), B=Constant() {r=1}, Radius=Constant() {r=0.1}, Hardness=Constant() {r=1})
  n17 = Subtract(A=ScalarParameter() {defaultValue=1, parameterName=Redlight_PuslingSpeed})
  n18 = LinearInterpolate(
   A=LinearInterpolate(
    B=ScalarParameter() {defaultValue=40, parameterName=EmissiveMult_4thLight}
    Alpha=MaterialFunctionCall(Speed=@n17) {materialFunction=&DebugTimeSine}.Result
   ) {constA=0.5, constB=0}
   B=ScalarParameter() {defaultValue=8, parameterName=EmissiveMult}
   Alpha=@n8
  )
  n19 = Fresnel() {baseReflectFraction=0.5, exponent=1}
  Multiply(A=Multiply(A=@n12, B=Multiply(A=@n16, B=@n18)), B=Power(Base=OneMinus(None=@n19)) {constExponent=1.5})
  Reroute(None=@n10.RGB)
  Subtract(A=@n2)
 ]
}
